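# Sound cue for the orb charge of a PistolCanary event tentacle attack (tac2100, Attack01).
# A single Wave Player plays one orb-charge wave at reduced volume, with no randomization or layering.
# It uses the tentacle-tip attenuation and the Fort_Events sound class.
# Bus sends feed sidechain buses for weapon volume and music high-pass and low-pass filtering, which suggests the charge ducks or filters those mixes.
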SoundCue /PistolCanary_Event_Audio/Assets/Section02/tacaud2100/SC_PistolCanary_Section01_tac2100_Attack01_Charge
root=WavePlayer_1
nodes [{id=WavePlayer_1, properties {soundWaveAssetPtr=/PistolCanary_Event_Audio/Assets/Section02/wav/PistolCanary_Tentacle_Attk_2100_OrbCharge_01}}]
properties {
 attenuationSettings=/PistolCanary_Audio/Assets/Settings/ATT_PistolCanary_Tentacle_Tip
 maxDistance=100000
 soundClassObject=&Fort_Events
 volumeMultiplier=0.7
 busSends [
  {audioBus=&AB_Sidechain_FortWeapon_Vol, minSendLevel=0, sendLevel=0.2, soundSourceBus=null}
  {audioBus=/Game/Sounds/Modulation/EnvelopeFollowers/HPF/AB_SideChain_FortMusic_HPF, minSendLevel=0, sendLevel=0.5, soundSourceBus=null}
  {audioBus=/Game/Sounds/Modulation/EnvelopeFollowers/LPF/AB_SideChain_FortMusic_LPF, minSendLevel=0, sendLevel=0.2, soundSourceBus=null}
 ]
}
