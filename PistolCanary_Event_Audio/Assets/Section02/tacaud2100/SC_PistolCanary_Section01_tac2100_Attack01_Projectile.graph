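# Sound cue for the projectile sound of the tac2100 tentacle orb-charge attack in the PistolCanary event.
# A single wave player plays the orb-charge projectile wave directly, with no randomisation or modulation.
# It uses the shared PistolCanary projectile attenuation, a long max audible distance and the events sound class, at slightly reduced volume.
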SoundCue /PistolCanary_Event_Audio/Assets/Section02/tacaud2100/SC_PistolCanary_Section01_tac2100_Attack01_Projectile
root=WavePlayer_1
nodes [{
  id=WavePlayer_1
  properties {soundWaveAssetPtr=/PistolCanary_Event_Audio/Assets/Section02/wav/PistolCanary_Tentacle_Attk_2100_OrbCharge_Projectile_OS}
 }]
properties {
 attenuationSettings=/PistolCanary_Audio/Assets/Settings/ATT_PistolCanary_Projectiles
 maxDistance=57000
 soundClassObject=&Fort_Events
 volumeMultiplier=0.7
}
